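# Conversation database for the DiaryString_WidePuck NPC, defining its greeting and the dialogue branches it offers.
# A single entry point leads to a speech node with a default line plus contextual lines for a first-ever conversation, an NPC conversation-count check and a specific Peely Toon outfit, each with a priority.
# That speech fans out through link nodes to shared basic-services and NPC quest conversations, and to Season 38 story and DiaryString entries.
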ConversationDatabase /NPCLibrary/NPCs/DiaryString_WidePuck/NPCLibrary_Conversation_DiaryString_WidePuck
entry_points [{tag=Conversation.Entry.DiaryString_WidePuck, destinations [#1]}]
nodes [
 ConversationEntryPointNode_7 {
  id=#1
  class=ConversationEntryPointNode
  outputs [#2]
  properties {entryTag {tagName=Conversation.Entry.DiaryString_WidePuck}}
 }
 Speech_C_4 {
  id=#2
  class=Speech_C
  outputs [
   #3
   #4
   #5
   #6
   #7
  ]
  properties {
   nodeName=""
   generalConfig {
    defaultMessage="Glavin!"
    contextualMessages [
     {
      message="Oh good glavin! These reality fluctuations are going to invalidate so many of my hypotheses!"
      contextRequirements [{requirement {class=FortControllerRequirement_IsFirstEverConversationWithNPC}}]
      priority {curve {rowName=Tandem.Speech.Priority.FirstEverInteraction}}
     }
     {
      message="Glavin!"
      contextRequirements [{requirement {
         class=FortControllerRequirement_CheckNPCMatchConversationCount
         properties {nPCUniqueIdTag {tagName=AISpawnerData.Type.Tandem.DiaryString_WidePuck}}
        }}]
      priority {curve {rowName=None}}
     }
     {
      message=I can hardly comprehend the ramifications of such an improbable meeting...I suspect the issue is my mushy banana brain.
      contextRequirements [{requirement {
         class=FortControllerRequirement_HasCID
         properties {allowedCIds [/BRCosmetics/Athena/Items/Cosmetics/Characters/CID_A_355_Athena_Commando_M_PeelyToon]}
        }}]
      priority {value=2, curve {rowName=Tandem.Speech.Priority.OutfitReaction}}
     }
    ]
   }
  }
 }
 ConversationLinkNode_4 {
  id=#7
  class=ConversationLinkNode
  properties {remoteEntryTag {tagName=Conversation.Entry.Season38.DiaryString.P04}}
 }
 ConversationLinkNode_3 {
  id=#6
  class=ConversationLinkNode
  properties {remoteEntryTag {tagName=Conversation.Entry.Season38.Story.P02.Q05.01}}
 }
 ConversationLinkNode_2 {
  id=#5
  class=ConversationLinkNode
  properties {remoteEntryTag {tagName=Conversation.Entry.Season38.Story.P02.Q02.01}}
 }
 ConversationLinkNode_0 {id=#4, class=ConversationLinkNode, properties {remoteEntryTag {tagName=Conversation.Entry.BasicServices}}}
 ConversationLinkNode_1 {id=#3, class=ConversationLinkNode, properties {remoteEntryTag {tagName=Conversation.Entry.NPCLibrary.Quests}}}
]
